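# Unlit additive lobby FX material that makes mesh elements float, with a glowing two-color Fresnel look.
# Two time-driven sines with separate speeds, phased by Pivot Painter per-element random values, build a bobbing offset along each element's X axis.
# A Spawned parameter blends geometry from collapsed at the object position to full shape, unwinding a rotation about a diagonal axis.
# Fresnel blends the colors and, inverted, drives depth-faded opacity.
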Material /Game/UI/Foundation/Lobby/materials/M_FloatingFX
graph {
 properties {blendMode=BLEND_Additive, pixelDepthOffsetMode=PDOM_Legacy, shadingModel=MSM_Unlit, usedShadingModels=Unlit}
 nodes [
  n1 = MaterialFunctionCall() {materialFunction=/Engine/Functions/Engine_MaterialFunctions02/PivotPainter/PivotPainter_PerObjectData}
  n2 = Add(A=Multiply(A=Time(), B=ScalarParameter() {defaultValue=3.1658, parameterName=Speed}), B=@n1.Random Value Per Element)
  n3 = Add(A=Multiply(A=Time(), B=ScalarParameter() {defaultValue=8.1833, parameterName=Speed2}), B=@n1.Random Value Per Element)
  n4 = Subtract(A=Add(A=Sine(None=@n2), B=Sine(None=@n3))) {constB=2}
  n5 = ScalarParameter() {defaultValue=1, parameterName=Spawned}
  n6 = RotateAboutAxis(
   NormalizedRotationAxis=Normalize(VectorInput=Constant3Vector() {constant=<rgba 1 1 1 0>})
   RotationAngle=LinearInterpolate(A=ScalarParameter() {defaultValue=1, parameterName=StartRotation}, B=Constant(), Alpha=@n5)
   PivotPoint=ObjectPositionWS()
   Position=WorldPosition().XYZ
  )
  n7 = Subtract(A=ObjectPositionWS(), B=Add(A=WorldPosition().XYZ, B=@n6))
  n8 = OneMinus(None=@n5)
  Add(
   A=Multiply(A=Multiply(A=@n4, B=@n1.X-Axis Vector), B=ScalarParameter() {defaultValue=1.1849, parameterName=Distance})
   B=Add(A=Multiply(A=@n7, B=@n8), B=@n6)
  )
  n9 = Fresnel(ExponentIn=ScalarParameter() {defaultValue=0.1073, parameterName=Fresnel}) {baseReflectFraction=0}
  DepthFade(
   Opacity=Multiply(A=ScalarParameter() {defaultValue=1.2202, parameterName=Opacity}, B=OneMinus(None=@n9))
   FadeDistance=ScalarParameter() {defaultValue=5, parameterName=DepthFade}
  )
  LinearInterpolate(
   A=VectorParameter() {parameterName=Color1, defaultValue=<rgba 0.2924 9.4027 10 1>}.RGB
   B=VectorParameter() {parameterName=Color2, defaultValue=<rgba 0.1094 0.0804 0 1>}.RGB
   Alpha=@n9
  )
 ]
}
