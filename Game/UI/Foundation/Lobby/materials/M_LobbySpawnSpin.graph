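# Additive unlit mesh-particle material, named for a spinning spawn effect in the lobby.
# A panned beam-reveal texture, driven by the PannerReveal dynamic parameter, is sharpened by a power and multiplied by Fresnel to mix inner and outer colors.
# The Pass parameter blends blue and gold color pairs into emissive adjustments.
# A second mask uses a softer exponent, and Thickness WPO scales a vertex-normal offset.
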Material /Game/UI/Foundation/Lobby/materials/M_LobbySpawnSpin
graph {
 properties {bUsedWithMeshParticles=true, blendMode=BLEND_Additive, pixelDepthOffsetMode=PDOM_Legacy, shadingModel=MSM_Unlit, usedShadingModels=Unlit}
 nodes [
  n1 = MaterialFunctionCall(
   Brightness=ScalarParameter() {defaultValue=1.5, parameterName=Fresnel Brightness}
   Exponent=ScalarParameter() {defaultValue=2, parameterName=Fresnel Exponent}
  ) {materialFunction=&MF_Fresnel_FX}
  n2 = TextureSample!(
   UVs=Panner(Time=DynamicParameter() {defaultValue=<rgba 0 0.2 1 0>, paramNames [Thickness WPO, PannerReveal, Opacity Add, Pass]}.PannerReveal) {speedX=1}
   Tex
   Level
   Apply View MipBias
  ) {
   constMipValue=0
   mipValueMode=TMVM_MipLevel
   samplerType=SAMPLERTYPE_LinearColor
   texture=/Game/Effects/Fort_Effects/Textures/Tile/T_Raygun_BeamReveal_01
  }
  n3 = Multiply(A=@n1.Fresnel, B=Power(Base=@n2.R, Exp=ScalarParameter() {defaultValue=100, parameterName=bbbb_1}))
  n4 = LinearInterpolate(
   A=VectorParameter() {parameterName=Color Inner_Pass, defaultValue=<rgba 0.2875 0.3736 0.5 1>}.RGB
   B=VectorParameter() {parameterName=l, defaultValue=<rgba 0.37 0.7959 1 1>}.RGB
   Alpha=@n3
  )
  n5 = LinearInterpolate(
   A=VectorParameter() {parameterName=Color Inner, defaultValue=<rgba 0.5 0.2371 0.0202 1>}.RGB
   B=VectorParameter() {parameterName=Color Outer, defaultValue=<rgba 2.25 1.7264 0.029 1>}.RGB
   Alpha=@n3
  )
  n6 = DynamicParameter() {defaultValue=<rgba 0 0.2 1 0>, paramNames [Thickness WPO, PannerReveal, Opacity Add, Pass]}
  MaterialFunctionCall(Emissive=Multiply(A=ScalarParameter() {defaultValue=0.3, parameterName=Op}, B=LinearInterpolate(A=@n4, B=@n5, Alpha=@n6.Pass))) {materialFunction=&MF_FX_EmissiveAdjustments}
  n7 = ScalarParameter() {defaultValue=2, parameterName=bbbb}
  Multiply(A=ScalarParameter() {defaultValue=0.5, parameterName=Op_1}, B=Multiply(A=@n1.Fresnel, B=Power(Base=@n2.R, Exp=@n7)))
  Multiply(
   A=ScalarParameter() {defaultValue=-0.5, parameterName=Op_2}
   B=Multiply(
    A=VertexNormalWS()
    B=DynamicParameter() {defaultValue=<rgba 0 0.2 1 0>, paramNames [Thickness WPO, PannerReveal, Opacity Add, Pass]}.Thickness WPO
   )
  )
 ]
}
