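# Lobby floor material: a metallic Default Lit surface with screen-space reflections that fades toward an emissive falloff color away from the object's center.
# A sphere mask drives that fade, attenuating the tiled floor texture and blending roughness toward a constant.
# Four pad circles, projected in world XY from position parameters, combine into the normal and mask the base color.
# The mobile shading path drops the circles and normal, using scaled brightness and simpler roughness.
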Material /Game/UI/Foundation/Lobby/materials/M_Lobby_Floor
graph {
 properties {
  bScreenSpaceReflections=true
  bUseMaterialAttributes=true
  pixelDepthOffsetMode=PDOM_Legacy
  translucencyLightingMode=TLM_SurfacePerPixelLighting
  usedShadingModels=Default Lit
  translucentMultipleScatteringExtinction=<rgba 1 0.9278 0.6832 1>
 }
 nodes [
  n1 = ScalarParameter() {defaultValue=0.85, parameterName=Metalness}
  n2 = Power(
   Base=SphereMask(A=ObjectPositionWS(), B=WorldPosition().XYZ, Radius=Constant() {r=746.3272}, Hardness=Constant() {r=0.0837})
   Exp=Constant() {r=3}
  )
  n3 = ShadingPathSwitch(Default=@n2, Mobile=@n2) {desc=Migrated from FeatureLevelSwitch}
  n4 = Multiply(A=VectorParameter() {parameterName=FalloffColor, defaultValue=<rgba 0.003 0.0043 0.01 1>}.RGB, B=OneMinus(None=@n3))
  n5 = WorldPosition()
  n6 = ScalarParameter() {defaultValue=150, parameterName=Ring Radius}
  n7 = Divide(A=Subtract(A=@n5.XYZ, B=VectorParameter() {parameterName=position1, defaultValue=<rgba -125.7088 125304 0 0>}.RGB), B=@n6)
  n8 = TextureSample(UVs=ComponentMask(None=Subtract(A=@n7) {constB=-0.5}) {g=true, r=true}) {samplerType=SAMPLERTYPE_LinearColor, texture=/Game/UI/Foundation/Lobby/textures/T_Lobby_PadFloorCircle}
  n9 = Multiply(A=@n8.RGB) {constB=2}
  n10 = Divide(A=Subtract(A=@n5.XYZ, B=VectorParameter() {parameterName=position2, defaultValue=<rgba -5.8829 125413 0 0>}.RGB), B=@n6)
  n11 = TextureSample(UVs=ComponentMask(None=Subtract(A=@n10) {constB=-0.5}) {g=true, r=true}) {samplerType=SAMPLERTYPE_LinearColor, texture=&T_Lobby_PadFloorCircle}
  n12 = Multiply(A=@n11.RGB) {constB=2}
  n13 = Divide(A=Subtract(A=@n5.XYZ, B=VectorParameter() {parameterName=position3, defaultValue=<rgba 126.7903 125311 0 0>}.RGB), B=@n6)
  n14 = TextureSample(UVs=ComponentMask(None=Subtract(A=@n13) {constB=-0.5}) {g=true, r=true}) {samplerType=SAMPLERTYPE_LinearColor, texture=&T_Lobby_PadFloorCircle}
  n15 = Add(A=Add(A=Add(A=@n9) {constB=-1}, B=Add(A=@n12) {constB=-1}), B=Add(A=Multiply(A=@n14.RGB) {constB=2}) {constB=-1})
  n16 = Divide(A=Subtract(A=@n5.XYZ, B=VectorParameter() {parameterName=position4, defaultValue=<rgba 257.929 125200 0 0>}.RGB), B=@n6)
  n17 = TextureSample(UVs=ComponentMask(None=Subtract(A=@n16) {constB=-0.5}) {g=true, r=true}) {samplerType=SAMPLERTYPE_LinearColor, texture=&T_Lobby_PadFloorCircle}
  n18 = Add(A=Multiply(A=@n17.RGB) {constB=2}) {constB=-1}
  n19 = Multiply(A=Multiply(A=@n8.B, B=@n11.B), B=Multiply(A=@n14.B, B=@n17.B))
  n20 = LinearInterpolate(
   A=Multiply(A=VectorParameter() {parameterName=TotalNormalAmt, defaultValue=<rgba 1 1 0.2 0>}.RGB, B=Add(A=@n15, B=@n18))
   B=Constant3Vector() {constant=<rgba 0 0 1 0>}
   Alpha=Power(Base=@n19, Exp=ScalarParameter() {defaultValue=10, parameterName=CircMaskPowA})
  )
  n21 = TextureSample(UVs=Multiply(A=TextureCoordinate(), B=ScalarParameter() {defaultValue=2.5, parameterName=Ugh_3})) {texture=/Game/UI/Foundation/Lobby/textures/T_Lobby_Floor_C}
  n22 = Power(Base=@n21.B, Exp=Constant() {r=2})
  n23 = OneMinus(None=@n19)
  n24 = ScalarParameter() {defaultValue=2, parameterName=CircMaskPowA_1}
  n25 = Multiply(A=@n22, B=OneMinus(None=Power(Base=@n23, Exp=@n24)))
  n26 = Multiply(
   A=@n3
   B=Multiply(
    A=ShadingPathSwitch(Default=@n25, Mobile=@n22) {desc=Migrated from FeatureLevelSwitch}
    B=VectorParameter() {parameterName=DiffuseColor01, defaultValue=<rgba 1.5 1.5 1.5 1>}.RGB
   )
  )
  n27 = ScalarParameter() {defaultValue=0.8, parameterName=RingRough_1}
  n28 = TextureCoordinate() {uTiling=1.6, vTiling=1.6}
  n29 = Clamp(None=LinearInterpolate(
    A=ScalarParameter() {defaultValue=-0.4, parameterName=SpecB}
    B=ScalarParameter() {defaultValue=2, parameterName=SpecA}
    Alpha=TextureSample(UVs=@n28) {samplerType=SAMPLERTYPE_Masks, texture=/Game/UI/Foundation/Lobby/textures/T_Lobby_Floor_S}.B
   ))
  n30 = Power(Base=@n3, Exp=ScalarParameter() {defaultValue=0.15, parameterName=FalloffRateRoughDist})
  n31 = ScalarParameter() {parameterName=SpecA_1}
  n32 = ScalarParameter() {defaultValue=0.5048, parameterName=BrightnessMoble}
  n33 = ScalarParameter() {defaultValue=0.9722, parameterName=RoughnessM1}
  n34 = ScalarParameter() {defaultValue=0.4751, parameterName=RoughnessM2}
  ShadingPathSwitch(
   Default=MakeMaterialAttributes!(
    BaseColor=@n26
    Metallic=@n1
    Specular=@n31
    Roughness=LinearInterpolate(A=@n27, B=@n29, Alpha=@n30)
    Anisotropy
    EmissiveColor=@n4
    Opacity
    OpacityMask
    Normal=Normalize(VectorInput=@n20)
    Tangent
    WorldPositionOffset
    SubsurfaceColor
    ClearCoat
    ClearCoatRoughness
    AmbientOcclusion
    Refraction
    CustomizedUVs_0
    CustomizedUVs_1
    CustomizedUVs_2
    CustomizedUVs_3
    CustomizedUVs_4
    CustomizedUVs_5
    CustomizedUVs_6
    CustomizedUVs_7
    PixelDepthOffset
    ShadingModel
    Displacement
   )
   Mobile=MakeMaterialAttributes(
    BaseColor=Multiply(A=@n26, B=@n32)
    Metallic=@n1
    Specular=@n31
    Roughness=LinearInterpolate(A=@n33, B=@n34, Alpha=@n21.B)
    EmissiveColor=@n4
   )
  ) {desc=Migrated from FeatureLevelSwitch}
 ]
}
