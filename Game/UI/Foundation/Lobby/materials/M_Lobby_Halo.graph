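# Unlit additive halo material for the lobby, built from procedural UV gradients with no textures.
# A vertical UV gradient shaped by a power feeds a DepthFade with an adjustable Fade distance, which softens intersections with geometry.
# A separate SphereMask around UV center is shaped by a FalloffRate power, alongside a Color parameter.
# The graph shows no links joining these branches or wiring them to material outputs.
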Material /Game/UI/Foundation/Lobby/materials/M_Lobby_Halo
graph {
 properties {
  blendMode=BLEND_Additive
  pixelDepthOffsetMode=PDOM_Legacy
  shadingModel=MSM_Unlit
  translucencyLightingMode=TLM_SurfacePerPixelLighting
  usedShadingModels=Unlit
  translucentMultipleScatteringExtinction=<rgba 1 0.9278 0.6832 1>
 }
 nodes [
  Constant() {r=100}
  n1 = TextureCoordinate()
  DepthFade(
   Opacity=Power(Base=ComponentMask(None=@n1) {g=true}, Exp=ScalarParameter() {defaultValue=2.8382})
   FadeDistance=ScalarParameter() {defaultValue=555, parameterName=Fade}
  )
  Power(
   Base=SphereMask(A=TextureCoordinate(), B=Constant() {r=0.5}, Hardness=ScalarParameter() {defaultValue=2.8382}) {attenuationRadius=2.5, hardnessPercent=1}
   Exp=ScalarParameter() {defaultValue=2.4149, parameterName=FalloffRate}
  )
  VectorParameter() {parameterName=Color, defaultValue=<rgba 0.0878 0.1296 0.225 1>}
 ]
}
