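# Clear-coat material for the floating lobby pad, with emissive looks for occupied, selected and group-colored states.
# A packed mask blends edge-to-center brightness and colors, desaturated when unoccupied, then boosted on selection with a mobile shading-path fallback, alongside ring and jet glows.
# Spawned drives position math that collapses vertices toward the pivot and rotates them.
# The graph also contains icon compositing with shifted, scaled UVs and opposing panned gradients.
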Material /Game/UI/Foundation/Lobby/materials/M_Lobby_Pad_Floating
graph {
 properties {
  bScreenSpaceReflections=true
  bUsedWithNiagaraMeshParticles=true
  numCustomizedUVs=2
  pixelDepthOffsetMode=PDOM_Legacy
  shadingModel=MSM_ClearCoat
  translucencyLightingMode=TLM_SurfacePerPixelLighting
  usedShadingModels=Clear Coat
  translucentMultipleScatteringExtinction=<rgba 1 0.9278 0.6832 1>
 }
 nodes [
  n1 = TextureSample() {texture=/Game/UI/Foundation/Lobby/textures/T_LobbyLightDisc_Floating_E}
  n2 = Multiply(
   A=VectorParameter() {parameterName=GroupColor, defaultValue=<rgba 0.5333 0.7379 0.0382 1>}.RGB
   B=ScalarParameter() {parameterName=GroupColorBrightnessIndicator}
  )
  n3 = VectorParameter() {parameterName=EmissiveColorX, defaultValue=<rgba 0 0.9319 1 1>}
  n4 = TextureSample() {samplerType=SAMPLERTYPE_Masks, texture=/Game/UI/Foundation/Lobby/textures/T_Lobby_Character_PadSmall_E_LargeParties}
  n5 = LinearInterpolate(
   A=LinearInterpolate(
    A=ScalarParameter() {defaultValue=1, group=NotDynamic, parameterName=BrightnessEdgeUnoccupied}
    B=ScalarParameter() {defaultValue=6, group=NotDynamic, parameterName=BrightnessCenterUnoccupied}
    Alpha=@n4.R
   )
   B=LinearInterpolate(
    A=ScalarParameter() {defaultValue=1.5, group=NotDynamic, parameterName=BrightnessEdge}
    B=ScalarParameter() {defaultValue=10, group=NotDynamic, parameterName=BrightnessCenter}
    Alpha=@n4.R
   )
   Alpha=ScalarParameter() {group=PrimaryState, parameterName=Occupied}
  )
  n6 = ScalarParameter() {defaultValue=3, parameterName=GlowRing}
  n7 = LinearInterpolate(A=OneMinus(None=@n4.R), B=Constant() {r=1}, Alpha=ScalarParameter() {group=PrimaryState, parameterName=Selected})
  n8 = Multiply(
   A=VectorParameter() {parameterName=GroupColor, defaultValue=<rgba 0.5333 0.7379 0.0382 1>}.RGB
   B=Add(A=Multiply(A=@n4.G, B=@n6), B=Multiply(A=@n4.B, B=@n7))
  )
  n9 = Desaturation(
   None=LinearInterpolate(
    A=LinearInterpolate(B=@n3.RGB, Alpha=@n5)
    B=@n8
    Alpha=ScalarParameter() {defaultValue=1, group=PrimaryState, parameterName=UseGroupPadColoration}
   )
   Fraction=LinearInterpolate(
    A=ScalarParameter() {defaultValue=0.75, group=NotDynamic, parameterName=UnoccupiedDesatFactor}
    Alpha=ScalarParameter() {group=PrimaryState, parameterName=Occupied}
   ) {constB=0.125}
  )
  n10 = Multiply(A=Multiply(A=Add(A=@n4.B) {constB=0.5}, B=@n4.B), B=@n9)
  n11 = ScalarParameter() {defaultValue=0.01, parameterName=EmissiveOffState}
  n12 = TextureSample() {texture=/Game/UI/Foundation/Lobby/textures/T_LobbyLightDisc_Floating_D}
  n13 = VectorParameter() {parameterName=IconGradColor_RChan_A, defaultValue=<rgba 0.0078 0.2463 0.5 1>}
  n14 = VectorParameter() {parameterName=IconGradColor_RChan_B, defaultValue=<rgba 0.2283 0.8497 0.8594 1>}
  n15 = VectorParameter() {parameterName=GroupColor, defaultValue=<rgba 0.5333 0.7379 0.0382 1>}
  n16 = Add(A=@n15.RGB) {constB=0.25}
  n17 = Multiply(
   A=LinearInterpolate(
    A=LinearInterpolate(A=@n13.RGB, B=@n14.RGB, Alpha=@n4.R)
    B=LinearInterpolate(A=@n15.RGB, B=@n16, Alpha=@n4.R)
    Alpha=ScalarParameter() {defaultValue=1, group=PrimaryState, parameterName=UseGroupPadColoration}
   )
   B=@n4.B
  )
  n18 = LinearInterpolate(
   A=ScalarParameter() {defaultValue=0.75, group=NotDynamic, parameterName=UnoccupiedDesatFactor}
   Alpha=ScalarParameter() {group=PrimaryState, parameterName=Occupied}
  ) {constB=0.125}
  n19 = Multiply(A=@n12.RGB, B=Multiply(A=Desaturation(None=@n17, Fraction=@n18), B=Power(Base=@n9)))
  n20 = Constant() {r=0.3}
  n21 = ScalarParameter() {defaultValue=1.5, parameterName=SelectedEmissivePower}
  n22 = Power(Base=@n19, Exp=ShadingPathSwitch(Default=@n21, Deferred=@n21, Mobile=Constant() {r=1}) {desc=Migrated from FeatureLevelSwitch})
  n23 = ScalarParameter() {defaultValue=5, parameterName=SelectedEmissiveMult}
  n24 = ShadingPathSwitch(Default=@n23, Deferred=@n23, Mobile=Constant() {r=1}) {desc=Migrated from FeatureLevelSwitch}
  n25 = LinearInterpolate(
   A=LinearInterpolate(A=Multiply(A=@n10, B=@n11), B=Multiply(A=@n19, B=@n20), Alpha=ScalarParameter() {group=PrimaryState, parameterName=Occupied})
   B=Clamp(None=Multiply(A=@n22, B=@n24))
   Alpha=ScalarParameter() {group=PrimaryState, parameterName=Selected}
  )
  n26 = VertexColor()
  Add(
   A=Multiply(A=VectorParameter() {parameterName=JetColor, defaultValue=<rgba 0 18.6998 20 1>}.RGB, B=@n1.G)
   B=Add(A=Multiply(A=@n1.R, B=@n2), B=Multiply(A=@n25, B=@n26.R))
  )
  n27 = WorldPosition()
  n28 = ScalarParameter() {defaultValue=1, parameterName=Spawned}
  n29 = Multiply(A=Subtract(A=ObjectPositionWS(), B=@n27.XYZ), B=OneMinus(None=@n28))
  n30 = RotateAboutAxis(
   NormalizedRotationAxis=Normalize(VectorInput=Constant3Vector() {constant=<rgba 1 1 1 0>})
   RotationAngle=LinearInterpolate(A=ScalarParameter() {defaultValue=1, parameterName=StartRotation}, B=Constant(), Alpha=@n28)
   PivotPoint=ObjectPositionWS()
   Position=WorldPosition().XYZ
  )
  Add(A=@n29, B=@n30)
  Add(A=@n27.XYZ, B=@n30)
  Constant3Vector()
  n31 = ScalarParameter() {defaultValue=1, parameterName=Roughness2}
  n32 = TextureSample() {texture=/Game/UI/Foundation/Lobby/textures/T_LobbyLightDisc_Floating_S}
  LinearInterpolate(A=ScalarParameter() {defaultValue=0.4286, parameterName=Roughness1}, B=@n31, Alpha=@n32.B)
  n33 = ScalarParameter() {defaultValue=0.1, group=NotDynamic, parameterName=MidtoneBlendUnoccupied}
  n34 = ScalarParameter() {defaultValue=0.25, group=NotDynamic, parameterName=MidtoneBlend}
  n35 = ScalarParameter() {group=PrimaryState, parameterName=Occupied}
  n36 = TextureSampleParameter2D(UVs=TextureCoordinate() {coordinateIndex=1}) {parameterName=IconTexture, samplerType=SAMPLERTYPE_LinearColor, texture=/Game/UI/SocialPanel/Art/Textures/T_UI_Logo_Dragon}
  n37 = LinearInterpolate(A=LinearInterpolate(A=Constant(), B=@n8, Alpha=LinearInterpolate(A=@n33, B=@n34, Alpha=@n35) {constB=0.075}), B=@n8, Alpha=@n36.G) {constA=0.5}
  LinearInterpolate(A=@n8, B=LinearInterpolate(A=LinearInterpolate(B=@n37, Alpha=@n36.B), B=@n5, Alpha=@n36.R), Alpha=@n36.A)
  n38 = Multiply(A=TextureCoordinate(), B=ScalarParameter() {defaultValue=1, parameterName=GradientSize})
  n39 = Constant() {r=-1}
  n40 = Time()
  n41 = ScalarParameter() {defaultValue=-1, parameterName=GradientSpeed}
  n42 = Constant()
  n43 = TextureSample(UVs=Panner(Coordinate=Multiply(A=@n38, B=@n39), Time=@n40, Speed=AppendVector(A=@n41, B=@n42))) {samplerType=SAMPLERTYPE_LinearGrayscale, texture=/Game/UI/Foundation/Lobby/textures/T_Lobby_PadFloorGradient}
  n44 = Constant()
  n45 = TextureSample(UVs=Panner(Coordinate=@n38, Time=@n40, Speed=AppendVector(A=@n41, B=@n44))) {samplerType=SAMPLERTYPE_LinearGrayscale, texture=&T_Lobby_PadFloorGradient}
  LinearInterpolate(A=@n43.RGB, B=@n45.RGB)
  n46 = MaterialFunctionCall(Float2=TextureCoordinate()) {materialFunction=&BreakOutFloat2Components}
  n47 = ScalarParameter() {defaultValue=0.0188, group=NotDynamic, parameterName=Icon_X_Shift}
  n48 = ScalarParameter() {defaultValue=-0.1287, group=NotDynamic, parameterName=Icon_Y_Shift}
  n49 = ScalarParameter() {defaultValue=0.5, group=NotDynamic, parameterName=Icon_UniformScale}
  MaterialFunctionCall(UVs=AppendVector(A=Add(A=@n46.R, B=@n47), B=Add(A=@n46.G, B=@n48)), Texture Scale=AppendVector(A=@n49, B=@n49)) {materialFunction=&ScaleUVsByCenter}
  n50 = Add(A=@n43.RGB, B=@n45.RGB)
  n51 = ScalarParameter() {defaultValue=8, parameterName=GradientPower}
  Multiply(A=Reroute(), B=Multiply(A=Power(Base=@n50, Exp=@n51), B=ScalarParameter() {defaultValue=0.1, parameterName=GradientAmount}))
  Multiply(A=@n32.G, B=ScalarParameter() {defaultValue=0.95, parameterName=Metal})
  Multiply(A=@n29, B=Constant3Vector() {constant=<rgba 1 1 0 0>})
  Multiply(A=@n32.B, B=@n31)
  Multiply(
   A=LinearInterpolate(
    A=ScalarParameter() {defaultValue=0.4952, parameterName=BrightnessDiffuse}
    B=ScalarParameter() {defaultValue=1, parameterName=BrightnessDiffuse2}
    Alpha=@n4.B
   )
   B=@n12.RGB
  )
  Multiply(A=@n43.RGB, B=@n45.RGB)
  Reroute(None=@n32.A)
  ScalarParameter() {defaultValue=0.1, group=Clear Coat, parameterName=Clear Coat Roughness}
  TextureSample() {desc=For testing, samplerType=SAMPLERTYPE_LinearColor, texture=&T_UI_Logo_Dragon}
  TextureSample() {samplerType=SAMPLERTYPE_Normal, texture=/Game/UI/Foundation/Lobby/textures/T_LobbyLightDisc_Floating_N}
  TextureSample() {desc=For testing, texture=/Game/Global/Textures/Default/Blanks/T_Blank_White_ZeroAlpha_sRGBColor}
  VertexColor()
 ]
}
