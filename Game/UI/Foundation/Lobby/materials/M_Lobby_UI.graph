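# M_Lobby_UI is a translucent, unlit material in the lobby UI folder.
# Its graph holds a single texture sample of the 128-pixel missing hero perk icon, with no math or parameters.
# No connection from the sample to a material output is shown, so it appears to be a stub or placeholder.
# It uses the default material domain rather than User Interface.
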Material /Game/UI/Foundation/Lobby/materials/M_Lobby_UI
graph {
 properties {blendMode=BLEND_Translucent, pixelDepthOffsetMode=PDOM_Legacy, shadingModel=MSM_Unlit, usedShadingModels=Unlit}
 nodes [TextureSample() {texture=/Game/UI/Foundation/Textures/Icons/Heroes/Icon-Perk-Missing-128}]
}
