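# Unlit translucent material for a lobby wall in the UI, built around a constant near-black navy color and a soft gradient mask.
# The mask multiplies a depth fade, softening intersections with scene geometry over the "fade" distance, by the UV V coordinate raised to the "Y Power" exponent.
# This makes a vertical falloff in UV space, with both softness and curve exposed as parameters.
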Material /Game/UI/Foundation/Lobby/materials/M_Lobby_Wall
graph {
 properties {
  blendMode=BLEND_Translucent
  pixelDepthOffsetMode=PDOM_Legacy
  shadingModel=MSM_Unlit
  translucencyLightingMode=TLM_SurfacePerPixelLighting
  usedShadingModels=Unlit
  translucentMultipleScatteringExtinction=<rgba 1 0.9278 0.6832 1>
 }
 nodes [
  Constant3Vector() {constant=<rgba 0.0028 0.0041 0.01 1>}
  n1 = TextureCoordinate()
  Multiply(
   A=DepthFade(Opacity=Constant() {r=1}, FadeDistance=ScalarParameter() {defaultValue=50, parameterName=fade})
   B=Power(Base=ComponentMask(None=@n1) {g=true}, Exp=ScalarParameter() {defaultValue=0.5, parameterName=Y Power})
  )
 ]
}
