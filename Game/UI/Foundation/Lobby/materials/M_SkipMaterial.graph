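# M_SkipMaterial is a translucent unlit lobby material that draws scrolling scanlines and a pulsing barrier grid.
# Panned line texture plus minimum opacity drives a two-color blend and an opacity term shaped by vertex-red HeightLerp and two opposed depth fades.
# A separate branch adds sine-animated blocks from tiled grid masks, offset per cell by Ceil of the UVs.
# A Visible parameter scales an offset toward the object position, apparently to collapse the mesh.
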Material /Game/UI/Foundation/Lobby/materials/M_SkipMaterial
graph {
 properties {blendMode=BLEND_Translucent, pixelDepthOffsetMode=PDOM_Legacy, shadingModel=MSM_Unlit, usedShadingModels=Unlit}
 nodes [
  n1 = Panner(
   Coordinate=Multiply(A=TextureCoordinate(), B=ScalarParameter() {defaultValue=28.632, parameterName=ScanlineTiling})
   Speed=ScalarParameter() {defaultValue=-0.5, parameterName=ScalineSpeed}
  ) {bFractionalPart=true}
  n2 = Add(
   A=Multiply(
    A=TextureSample(UVs=@n1) {samplerType=SAMPLERTYPE_LinearColor, texture=/Game/Athena/SafeZone/Lines}.R
    B=ScalarParameter() {defaultValue=0.1, parameterName=ScanLineOpacity}
   )
   B=ScalarParameter() {defaultValue=0.5142, parameterName=MinOpacity}
  )
  n3 = Multiply(A=TextureCoordinate(), B=ScalarParameter() {defaultValue=2.5232, parameterName=BlocksTiling})
  n4 = TextureCoordinate()
  n5 = ScalarParameter() {defaultValue=1.6705, parameterName=BlocksTiling_2}
  n6 = Add(
   A=TextureSample(UVs=@n3) {samplerType=SAMPLERTYPE_Masks, texture=/Game/Athena/Playlists/Barrier/Textures/T_BarrierGrid_M}.R
   B=TextureSample(UVs=Multiply(A=@n4, B=@n5)) {samplerType=SAMPLERTYPE_Masks, texture=&T_BarrierGrid_M}.R
  )
  n7 = Time() {period=1}
  n8 = Ceil(None=Abs(None=@n3))
  n9 = Add(A=ComponentMask(None=@n8) {r=true}, B=ComponentMask(None=@n8) {g=true})
  n10 = Add(
   A=ComponentMask(None=@n6) {r=true}
   B=Multiply(A=Add(A=@n7, B=@n9), B=ScalarParameter() {defaultValue=0.1467, parameterName=Sine Speed Grid})
  )
  n11 = Multiply(A=Sine(None=Frac(None=@n10)), B=ComponentMask(None=@n6) {g=true})
  n12 = ScalarParameter() {defaultValue=5, parameterName=Grid Power}
  n13 = Divide(A=TextureCoordinate(), B=ScalarParameter() {defaultValue=0.3563, parameterName=BlocksTiling_1})
  Add(
   A=@n2
   B=Multiply(
    A=Power(Base=@n11, Exp=@n12)
    B=TextureSample(UVs=@n13) {samplerType=SAMPLERTYPE_Masks, texture=&T_BarrierGrid_M}.B
   )
  )
  n14 = VertexNormalWS()
  n15 = CameraVectorWS()
  n16 = Subtract(A=SceneDepth().R, B=PixelDepth().R)
  n17 = ScalarParameter() {defaultValue=118.6953, parameterName=Distance Divisor}
  Clamp(None=Multiply(A=DotProduct(A=@n14, B=@n15), B=Divide(A=@n16, B=@n17)))
  Constant() {r=0.261}
  LinearInterpolate(
   A=VectorParameter() {parameterName=Color, defaultValue=<rgba 0.2917 0.2917 0.2917 1>}.RGB
   B=VectorParameter() {parameterName=Color2, defaultValue=<rgba 0.4792 0.4697 0.4692 1>}.RGB
   Alpha=@n2
  )
  Multiply(A=Subtract(A=ObjectPositionWS(), B=WorldPosition().XYZ), B=OneMinus(None=ScalarParameter() {defaultValue=1, parameterName=Visible}))
  n18 = ScalarParameter() {defaultValue=2.9189, parameterName=Opacity}
  n19 = VertexColor()
  n20 = Power(Base=DepthFade(Opacity=Constant() {r=1}, FadeDistance=@n17), Exp=ScalarParameter() {defaultValue=1.0923, parameterName=Contrast})
  n21 = Constant() {r=1}
  n22 = ScalarParameter() {defaultValue=100, parameterName=FadeDistanceMask}
  n23 = Multiply(A=OneMinus(None=DepthFade(Opacity=@n21, FadeDistance=@n22)), B=ScalarParameter() {defaultValue=2, parameterName=Contrast2add})
  Multiply(
   A=Multiply(
    A=Multiply(A=@n2, B=@n18)
    B=MaterialFunctionCall(Transition Phase=@n19.R, Height Texture=@n2) {materialFunction=&HeightLerp}.Alpha
   )
   B=Multiply(A=Clamp(None=@n20), B=Clamp(None=@n23))
  )
 ]
}
