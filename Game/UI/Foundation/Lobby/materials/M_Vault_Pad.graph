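# Lit material for the lobby vault character pad, with separate desktop and mobile tuning through shading path switches.
# Vertex color green blends two tint colors, multiplied by a mask texture, then an E-texture mask blends in a third color before brightness scaling.
# The same channel blends metallic and mobile roughness, and flattens the normal map on the default path.
# Default roughness follows a pad texture's red channel, while specular is a per-path scalar.
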Material /Game/UI/Foundation/Lobby/materials/M_Vault_Pad
graph {
 properties {
  bScreenSpaceReflections=true
  pixelDepthOffsetMode=PDOM_Legacy
  translucencyLightingMode=TLM_SurfacePerPixelLighting
  usedShadingModels=Default Lit
  translucentMultipleScatteringExtinction=<rgba 1 0.9278 0.6832 1>
 }
 nodes [
  n1 = VertexColor()
  n2 = TextureSample() {samplerType=SAMPLERTYPE_Masks, texture=/Game/UI/Foundation/Lobby/textures/T_Lobby_Character_Pad_S}
  Multiply(A=@n2.G, B=OneMinus(None=@n1.G))
  Reroute(None=LinearInterpolate(
    A=ScalarParameter() {defaultValue=0.1, group=Specular, parameterName=SpecRoughnessMin}
    B=ScalarParameter() {defaultValue=0.9, group=Specular, parameterName=SpecRoughnessMax}
    Alpha=@n2.B
   ))
  Reroute(None=@n2.A)
  n3 = VectorParameter() {parameterName=Color2, defaultValue=<rgba 0.0925 0.1041 0.1146 1>}
  n4 = VectorParameter() {parameterName=Color1, defaultValue=<rgba 0.1058 0.1165 0.1302 1>}
  n5 = VertexColor()
  n6 = LinearInterpolate(
   A=Multiply(A=LinearInterpolate(A=@n3.RGB, B=@n4.RGB, Alpha=@n5.G), B=TextureSample() {texture=/Game/UI/Foundation/Lobby/textures/T_VaultPad_M}.R)
   B=VectorParameter() {parameterName=Color3, defaultValue=<rgba 0.4484 0.45 0.4531 1>}.RGB
   Alpha=TextureSample() {texture=/Game/UI/Foundation/Lobby/textures/T_VaultPad_E}.R
  )
  ShadingPathSwitch(
   Default=Multiply(A=@n6, B=ScalarParameter() {defaultValue=0.9378, parameterName=Brightness_PC})
   Mobile=Multiply(A=@n6, B=ScalarParameter() {defaultValue=1, parameterName=Brightness_M})
  ) {desc=Migrated from FeatureLevelSwitch}
  ShadingPathSwitch(
   Default=LinearInterpolate(
    A=ScalarParameter() {defaultValue=0.2092, parameterName=Metall_PC}
    B=ScalarParameter() {defaultValue=0.3058, parameterName=Metall_PC_1}
    Alpha=@n1.G
   )
   Mobile=LinearInterpolate(A=ScalarParameter() {defaultValue=0.7637, parameterName=Metall_MObile}, B=ScalarParameter() {parameterName=Metall_MObile_1}, Alpha=@n1.G)
  ) {desc=Migrated from FeatureLevelSwitch}
  ShadingPathSwitch(
   Default=ScalarParameter() {defaultValue=0.5, parameterName=Specular_PC}
   Mobile=ScalarParameter() {defaultValue=0.01, parameterName=Specular_Mobile}
  ) {desc=Migrated from FeatureLevelSwitch}
  ShadingPathSwitch(
   Default=LinearInterpolate(
    A=ScalarParameter() {defaultValue=0.2373, group=Specular, parameterName=RoughnessPad_1}
    B=ScalarParameter() {defaultValue=0.4917, group=Specular, parameterName=RoughnessPad}
    Alpha=TextureSample() {texture=/Game/UI/Foundation/Lobby/textures/T_Lobby_Character_Pad_C}.R
   ) {constB=0.4}
   Mobile=LinearInterpolate(
    A=ScalarParameter() {defaultValue=0.565, parameterName=Roughness_Mobile}
    B=ScalarParameter() {defaultValue=0.3, parameterName=Roughness_Mobile_1}
    Alpha=@n5.G
   )
  ) {desc=Migrated from FeatureLevelSwitch}
  n7 = TextureSample() {samplerType=SAMPLERTYPE_Normal, texture=/Game/UI/Foundation/Lobby/textures/T_Lobby_Character_Pad_N}
  ShadingPathSwitch(Default=LinearInterpolate(A=@n7.RGB, B=Constant3Vector() {constant=<rgba 0 0 1 0>}, Alpha=@n5.G), Mobile=@n7.RGB) {desc=Migrated from FeatureLevelSwitch}
 ]
}
