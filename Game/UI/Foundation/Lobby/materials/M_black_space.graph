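# M_black_space is an unlit material in the lobby UI folder, apparently a near-black space backdrop.
# The graph holds only a very dark navy color constant and a scalar constant left at its default, with no textures, parameters or math.
# The file does not show which material outputs these constants drive.
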Material /Game/UI/Foundation/Lobby/materials/M_black_space
graph {
 properties {pixelDepthOffsetMode=PDOM_Legacy, shadingModel=MSM_Unlit, usedShadingModels=Unlit}
 nodes [
  Constant3Vector() {constant=<rgba 0.0028 0.0041 0.01 1>}
  Constant()
 ]
}
